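# Physics asset for the GrumbleCroak backpack cosmetic, defining physics bodies for its dangling frog-head rope.
# Four sphere bodies follow the dyn_froghead_rope bones, with a noticeably larger sphere on the last link.
# Constraints chain each link to the previous one and the first to the pack, using identical limited swing and tighter twist.
# The pack body uses two large offset spheres to approximate its volume.
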PhysicsAsset /BRCosmetics/Accessories/FORT_Backpacks/Backpack_M_MED_GrumbleCroak/Meshes/M_MED_GrumbleCroak_Pack_Physics
bodies {
 dyn_froghead_rope_1 [{shapeType=Sphere, radius=2, shapeName=dyn_froghead_rope_1_sphere, center=<0.8744 0 0>}]
 dyn_froghead_rope_2 [{shapeType=Sphere, radius=2, shapeName=dyn_froghead_rope_2_sphere, center=<0.8744 0 0>}]
 dyn_froghead_rope_3 [{shapeType=Sphere, radius=2, shapeName=dyn_froghead_rope_3_sphere, center=<0.8744 0 0>}]
 dyn_froghead_rope_4 [{shapeType=Sphere, radius=5.5, shapeName=dyn_froghead_rope_4_sphere, center=<3.2247 0 -0.3734>}]
 pack [
  {shapeType=Sphere, radius=23.5533, center=<0 18.9889 -17.9624>}
  {shapeType=Sphere, radius=14.5, shapeName=pack_sphere, center=<0 -8.3304 -5.3854>}
 ]
}
constraints [
 {bone1Name=dyn_froghead_rope_1, bone2Name=pack, swing1LimitDegrees=30, swing2LimitDegrees=30, twistLimitDegrees=10}
 {bone1Name=dyn_froghead_rope_2, bone2Name=dyn_froghead_rope_1, swing1LimitDegrees=30, swing2LimitDegrees=30, twistLimitDegrees=10}
 {bone1Name=dyn_froghead_rope_3, bone2Name=dyn_froghead_rope_2, swing1LimitDegrees=30, swing2LimitDegrees=30, twistLimitDegrees=10}
 {bone1Name=dyn_froghead_rope_4, bone2Name=dyn_froghead_rope_3, swing1LimitDegrees=30, swing2LimitDegrees=30, twistLimitDegrees=10}
]
