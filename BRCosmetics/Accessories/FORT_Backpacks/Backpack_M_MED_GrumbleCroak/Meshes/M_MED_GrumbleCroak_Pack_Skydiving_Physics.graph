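# Physics asset for the GrumbleCroak backpack cosmetic's skydiving setup, defining physics for the rope bones named for the frog head.
# A single large sphere on the pack bone anchors a chain of four small sphere bodies along the rope bones.
# Constraints link each rope segment to the one before it with identical limited swing and tight twist, so the rope can bend without twisting much.
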PhysicsAsset /BRCosmetics/Accessories/FORT_Backpacks/Backpack_M_MED_GrumbleCroak/Meshes/M_MED_GrumbleCroak_Pack_Skydiving_Physics
bodies {
 dyn_froghead_rope_1 [{shapeType=Sphere, radius=3, shapeName=dyn_froghead_rope_1_sphere}]
 dyn_froghead_rope_2 [{shapeType=Sphere, radius=3, shapeName=dyn_froghead_rope_2_sphere}]
 dyn_froghead_rope_3 [{shapeType=Sphere, radius=3, shapeName=dyn_froghead_rope_3_sphere}]
 dyn_froghead_rope_4 [{shapeType=Sphere, radius=3, shapeName=dyn_froghead_rope_4_sphere}]
 pack [{shapeType=Sphere, radius=23.5533, center=<0 -8.2706 -9.376>}]
}
constraints [
 {bone1Name=dyn_froghead_rope_1, bone2Name=pack, swing1LimitDegrees=30, swing2LimitDegrees=40, twistLimitDegrees=10}
 {bone1Name=dyn_froghead_rope_2, bone2Name=dyn_froghead_rope_1, swing1LimitDegrees=30, swing2LimitDegrees=40, twistLimitDegrees=10}
 {bone1Name=dyn_froghead_rope_3, bone2Name=dyn_froghead_rope_2, swing1LimitDegrees=30, swing2LimitDegrees=40, twistLimitDegrees=10}
 {bone1Name=dyn_froghead_rope_4, bone2Name=dyn_froghead_rope_3, swing1LimitDegrees=30, swing2LimitDegrees=40, twistLimitDegrees=10}
]
