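# Animation Blueprint for the GrumbleCroak backpack, adding secondary motion to its frog head, tongue, handle and knives.
# A cached input pose feeds skydiving and normal RigidBody simulations, blended on bIsSkyDiving.
# Additive bone offsets, mostly bool-toggled, then adjust that result before low-alpha AnimDynamics finish the handle, tongue and knives.
# The event graph updates skydive and parachute tail flags and computes the RigidBody gravity override through a deprecated state function with wind parameters.
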AnimBlueprint /BRCosmetics/Accessories/FORT_Backpacks/Backpack_M_MED_GrumbleCroak/Meshes/M_MED_GrumbleCroak_Pack_AnimBP
target_skeleton=&/M_MED_GrumbleCroak_Pack_Skeleton
graphs [
 AnimGraph {nodes [
   "Transform (Modify) Bone - Bone: dyn_froghead_rope_MSTR"(Alpha=1, Translation="0.4,0,0", Rotation="7.695,32.423,16.281", Scale="1,1,1") {
    rotationMode=BMM_Additive
    translationMode=BMM_Additive
    alphaBoolBlend {blendOption=Linear}
    boneToModify {boneName=dyn_froghead_rope_MSTR}
    rotation=<rot 7.695 32.423 16.281>
    translation {x=0.4}
   }
   n1 = Anim Dynamics - Bone: dyn_backpack_handle_1(ComponentPose=Local To Component(LocalPose=Use cached pose 'Transform Setup'()), Alpha=0.2, ExternalForce="0,0,0") {
    angularDampingOverride=0.9999
    bOverrideAngularDamping=true
    bUseGravityOverride=true
    gravityScale=0
    simulationSpace=BoneRelative
    alphaBoolBlend {blendOption=Linear}
    boundBone {boneName=dyn_backpack_handle_1}
    componentAppliedLinearAccClamp=<2000 2000 2000>
    componentLinearAccScale=<0.1 0.1 0.1>
    componentLinearVelScale=<0.1 0.1 0.1>
    gravityOverride {z=800}
    physicsBodyDefinitions [{
      boundBone {boneName=dyn_backpack_handle_1}
      boxExtents=<10 10 10>
      constraintSetup {angularLimitsMax=<0 0 5>, angularLimitsMin=<0 0 -15>, angularTarget=<0 0 0>}
      localJointOffset=<10 0 0>
     }]
   }
   n2 = Anim Dynamics - Bone: dyn_backpack_handle_2(ComponentPose=@n1, Alpha=0.4, ExternalForce="0,0,0") {
    angularDampingOverride=0.9999
    bOverrideAngularDamping=true
    bUseGravityOverride=true
    gravityScale=0
    simulationSpace=BoneRelative
    alphaBoolBlend {blendOption=Linear}
    boundBone {boneName=dyn_backpack_handle_2}
    componentAppliedLinearAccClamp=<2000 2000 2000>
    componentLinearAccScale=<0.1 0.1 0.1>
    componentLinearVelScale=<0.1 0.1 0.1>
    gravityOverride {z=800}
    physicsBodyDefinitions [{
      boundBone {boneName=dyn_backpack_handle_2}
      boxExtents=<10 10 10>
      constraintSetup {angularLimitsMax=<0 0 5>, angularLimitsMin=<0 0 -15>, angularTarget=<0 0 0>}
      localJointOffset=<10 0 0>
     }]
   }
   n3 = Anim Dynamics - Bone: dyn_froghead_tongue_1(ComponentPose=@n2, Alpha=0.3, ExternalForce="0,0,0") {
    gravityScale=0
    simulationSpace=BoneRelative
    alphaBoolBlend {blendOption=Linear}
    boundBone {boneName=dyn_froghead_tongue_1}
    componentAppliedLinearAccClamp=<2000 2000 2000>
    componentLinearAccScale=<0.3 0.3 0.3>
    componentLinearVelScale=<0.3 0.3 0.3>
    gravityOverride {z=-980}
    physicsBodyDefinitions [{
      boundBone {boneName=dyn_froghead_tongue_1}
      boxExtents=<10 10 10>
      constraintSetup {angularLimitsMax=<0 10 2>, angularLimitsMin=<0 -30 -2>, angularTarget=<0 0 0>}
      localJointOffset=<10 0 0>
     }]
   }
   n4 = Anim Dynamics - Bone: dyn_froghead_tongue_2(ComponentPose=@n3, Alpha=0.3, ExternalForce="0,0,0") {
    gravityScale=0
    simulationSpace=BoneRelative
    alphaBoolBlend {blendOption=Linear}
    boundBone {boneName=dyn_froghead_tongue_2}
    componentAppliedLinearAccClamp=<2000 2000 2000>
    componentLinearAccScale=<0.3 0.3 0.3>
    componentLinearVelScale=<0.3 0.3 0.3>
    gravityOverride {z=-980}
    physicsBodyDefinitions [{
      boundBone {boneName=dyn_froghead_tongue_2}
      boxExtents=<10 10 10>
      constraintSetup {angularLimitsMax=<0 30 2>, angularLimitsMin=<0 -30 -2>, angularTarget=<0 0 0>}
      localJointOffset=<10 0 0>
     }]
   }
   n5 = Anim Dynamics - Bone: dyn_knife_r(ComponentPose=@n4, Alpha=0.05, ExternalForce="0,0,0") {
    angularDampingOverride=1
    angularSpringConstant=100000
    bAngularSpring=true
    bOverrideAngularDamping=true
    simulationSpace=BoneRelative
    alphaBoolBlend {blendOption=Linear}
    boundBone {boneName=dyn_knife_r}
    componentAppliedLinearAccClamp=<2000 2000 2000>
    componentLinearAccScale=<0.2 0.2 0.2>
    componentLinearVelScale=<0.2 0.2 0.2>
    physicsBodyDefinitions [{
      boundBone {boneName=dyn_knife_r}
      boxExtents=<20 20 20>
      constraintSetup {angularLimitsMax=<0 5 3>, angularLimitsMin=<0 -5 -6>, angularTarget=<0 0 0>}
      localJointOffset=<25 0 0>
     }]
   }
   n6 = Anim Dynamics - Bone: dyn_knife_l(ComponentPose=@n5, Alpha=0.05, ExternalForce="0,0,0") {
    angularDampingOverride=1
    angularSpringConstant=100000
    bAngularSpring=true
    bOverrideAngularDamping=true
    simulationSpace=BoneRelative
    alphaBoolBlend {blendOption=Linear}
    boundBone {boneName=dyn_knife_l}
    componentAppliedLinearAccClamp=<2000 2000 2000>
    componentLinearAccScale=<0.2 0.2 0.2>
    componentLinearVelScale=<0.2 0.2 0.2>
    physicsBodyDefinitions [{
      boundBone {boneName=dyn_knife_l}
      boxExtents=<20 20 20>
      constraintSetup {angularLimitsMax=<0 5 6>, angularLimitsMin=<0 -5 -1>, angularTarget=<0 0 0>}
      localJointOffset=<30 0 0>
     }]
   }
   Output Pose(Result=Component To Local(ComponentPose=@n6)) {layerGroup=None}
   n7 = "Transform (Modify) Bone - Bone: pack"(ComponentPose=Local To Component(LocalPose=Use cached pose 'PostBlend'()), Translation="0,0.2,0", Rotation="0,0,0") {
    translationMode=BMM_Additive
    translationSpace=BCS_BoneSpace
    alphaBoolBlend {blendOption=Linear}
    boneToModify {boneName=pack}
    translation {y=0.2}
   }
   n8 = "Transform (Modify) Bone - Bone: pack"!(ComponentPose=@n7, bAlphaBoolEnabled:bool, Translation="0,-1,1", Rotation="-0.000,0.000,20.000", Scale) -> (Pose) {
    alphaInputType=Bool
    rotationMode=BMM_Additive
    rotationSpace=BCS_BoneSpace
    translationMode=BMM_Additive
    translationSpace=BCS_BoneSpace
    alphaBoolBlend {blendInTime=0.3, blendOutTime=0.3}
    boneToModify {boneName=pack}
    rotation {roll=20}
    translation {y=-1, z=1}
   }
   n9 = "Transform (Modify) Bone - Bone: dyn_froghead_rope_MSTR"!(ComponentPose=@n8, bAlphaBoolEnabled:bool=True, Translation="-0.5,0,0", Rotation="4.914,51.651,28.689", Scale="1,1,1") -> (Pose) {
    alphaInputType=Bool
    rotationMode=BMM_Additive
    rotationSpace=BCS_BoneSpace
    translationMode=BMM_Additive
    translationSpace=BCS_BoneSpace
    alphaBoolBlend {blendInTime=0.3, blendOption=ExpInOut, blendOutTime=0.3}
    boneToModify {boneName=dyn_froghead_rope_MSTR}
    rotation=<rot 4.914 51.651 28.689>
    translation {x=-0.5}
   }
   n10 = "Transform (Modify) Bone - Bone: dyn_froghead_rope_MSTR"(ComponentPose=@n9, bAlphaBoolEnabled=True, Translation="-0.5,0,0", Rotation="20,0,10", Scale="1,1,1") {
    alphaInputType=Bool
    rotationMode=BMM_Additive
    rotationSpace=BCS_BoneSpace
    translationMode=BMM_Additive
    translationSpace=BCS_BoneSpace
    alphaBoolBlend {blendInTime=0.3, blendOutTime=0.3}
    boneToModify {boneName=dyn_froghead_rope_MSTR}
    rotation {pitch=20, roll=10}
    translation {x=-0.5}
   }
   n11 = "Transform (Modify) Bone - Bone: dyn_backpack_handle_MSTR"!(ComponentPose=@n10, bAlphaBoolEnabled:bool=True, Translation="0,0,0", Rotation="0,-35,0", Scale="1,1,1") -> (Pose) {
    alphaInputType=Bool
    rotationMode=BMM_Additive
    rotationSpace=BCS_BoneSpace
    translationMode=BMM_Additive
    translationSpace=BCS_BoneSpace
    alphaBoolBlend {blendInTime=0.3, blendOption=ExpInOut, blendOutTime=0.3}
    boneToModify {boneName=dyn_backpack_handle_MSTR}
    rotation {yaw=-35}
   }
   n12 = "Transform (Modify) Bone - Bone: dyn_knife_r"!(ComponentPose=@n11, bAlphaBoolEnabled:bool=True, Translation="0,0,0", Rotation="-13,0,0", Scale="1,1,1") -> (Pose) {
    alphaInputType=Bool
    rotationMode=BMM_Additive
    rotationSpace=BCS_BoneSpace
    translationMode=BMM_Additive
    translationSpace=BCS_BoneSpace
    alphaBoolBlend {blendInTime=0.3, blendOption=ExpInOut, blendOutTime=0.3}
    boneToModify {boneName=dyn_knife_r}
    rotation {pitch=-13}
   }
   n13 = "Transform (Modify) Bone - Bone: dyn_knife_l"!(ComponentPose=@n12, bAlphaBoolEnabled:bool=True, Translation="0,0,0", Rotation="-13,0,0", Scale="1,1,1") -> (Pose) {
    alphaInputType=Bool
    rotationMode=BMM_Additive
    rotationSpace=BCS_BoneSpace
    translationMode=BMM_Additive
    translationSpace=BCS_BoneSpace
    alphaBoolBlend {blendInTime=0.3, blendOption=ExpInOut, blendOutTime=0.3}
    boneToModify {boneName=dyn_knife_l}
    rotation {pitch=-13}
   }
   Save cached pose 'Transform Setup'(Pose=Component To Local(ComponentPose=@n13)) {cacheName=Transform Setup}
   n14 = RigidBody!(
    ComponentPose=Local To Component(LocalPose=Use cached pose 'Input Pose'())
    Alpha
    OverrideWorldGravity:Vector=Get Gravity_Override() -> (Gravity_Override:Vector)
    ExternalForce
   ) -> (Pose) {
    bOverrideWorldGravity=true
    overridePhysicsAsset=&/M_MED_GrumbleCroak_Pack_Physics
    simulationSpace=BaseBoneSpace
    alphaBoolBlend {blendOption=Linear}
    baseBoneRef {boneName=root}
    componentAppliedLinearAccClamp=<4000 4000 4000>
    componentLinearAccScale=<0.6 0.6 0.6>
    componentLinearVelScale=<0.6 0.6 0.6>
   }
   Save cached pose 'No skydiving RB'(Pose=Component To Local(ComponentPose=@n14)) {cacheName=No skydiving RB}
   Save cached pose 'Input Pose'(Pose=Input Pose()) {cacheName=Input Pose}
   n15 = RigidBody(ComponentPose=Local To Component(LocalPose=Use cached pose 'Input Pose'()), OverrideWorldGravity=Get Gravity_Override()) {
    bOverrideWorldGravity=true
    overridePhysicsAsset=&/M_MED_GrumbleCroak_Pack_Skydiving_Physics
    simulationSpace=BaseBoneSpace
    alphaBoolBlend {blendOption=Linear}
    baseBoneRef {boneName=pack}
    componentAppliedLinearAccClamp=<5000 5000 5000>
    componentLinearAccScale=<0.8 0.8 0.8>
    componentLinearVelScale=<0.8 0.8 0.8>
   }
   Save cached pose 'Skydiving RB'(Pose=Component To Local(ComponentPose=@n15)) {cacheName=Skydiving RB}
   Save cached pose 'PostBlend'(Pose=Blend Poses by bool(
     BlendPose_0=Use cached pose 'Skydiving RB'()
     BlendPose_1=Use cached pose 'No skydiving RB'()
     BlendTime_0=0.3
     BlendTime_1=0.3
     bActiveValue=Get bIsSkyDiving()
    )) {cacheName=PostBlend}
  ]}
 EventGraph {nodes [
   Event BlueprintUpdateAnimation()
   > n1 = Sequence!(execute) -> (then_0, then_1, then_2:Exec)
   > Set IsSkydivingNoParachute(IsSkydivingNoParachute:bool=OR Boolean(A=Get bIsSkydiveDiveMode(), B=Get bIsSkydiveFloating())) -> (Output_Get:bool)
   n2 = Set CharacterHasTail(execute=@n1.then_1, CharacterHasTail:bool=CheckForTail!(self:Custom Character Part Anim Instance Object Reference) -> (bHasTail:bool)) -> (Output_Get:bool)
   > Set SkydiveWithTail(SkydiveWithTail:bool=AND Boolean(A=@n2, B=Get IsSkydivingNoParachute() -> (IsSkydivingNoParachute:bool))) -> (Output_Get:bool)
   > Set ParachuteWithTail(ParachuteWithTail:bool=AND Boolean(A=Get CharacterHasTail() -> (CharacterHasTail:bool), B=Get bIsParachuteOpened())) -> (Output_Get:bool)
   n3 = Gravity Override State Function DEPRECATED(
    execute=@n1.then_2
    CharacterPartAnimInstance=Self-Reference()
    GameTimeInSeconds=GetGameTimeInSeconds()
    GravityOverrideParams=Get Gravity Override Params()
   )
   > Set Gravity_Override(Gravity_Override:Vector=@n3) -> (Output_Get:Vector)
   TryGetPawnOwner()
  ]}
]
variables [
 CharacterHasTail {type=boolean}
 Gravity Override Params {
  type=GravityOverrideParamsStruct
  default {
   skydive_DivingDown {windAmplitude=<600 -1000 1000>, windFrequency=<6 10 10>, windOffset=<0 -800 800>}
   skydive_FloatingDown {gravityOverride=<0 0 1000>, windAmplitude=<500 -800 1000>, windFrequency=<6 10 10>, windOffset=<0 -400 500>}
  }
 }
 Gravity_Override {type=Vector}
 IsSkydivingNoParachute {type=boolean}
 ParachuteWithTail {type=boolean}
 SkydiveWithTail {type=boolean}
]
